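# Physics asset for the male medium RobedSledge Cross NPC body mesh, defining its simplified physics shapes and joints.
# One capsule per major bone covers pelvis, spine, neck, arms, hands, legs and feet, with no separate head or robe bodies.
# Constraints branch from the pelvis with uniform 45-degree swing and twist limit values.
# Calves attach directly to the pelvis, as siblings of the thigh-twist capsules, because the thigh bones have no bodies.
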PhysicsAsset /BRCosmetics/Characters/Player/Male/Medium/Bodies/M_MED_RobedSledge_Cross_NPC/Meshes/M_MED_RobedSledge_Cross_NPCBody_PhysicsAsset
bodies {
 calf_l [{shapeType=Capsule, length=21.3723, radius=8.4921, center=<-18.9348 -2.0577 -0.4065>, rotation=<rot 0 -92.5831 -88.4461>}]
 calf_r [{shapeType=Capsule, length=21.3723, radius=8.4921, center=<18.9348 2.0577 0.4065>, rotation=<rot 0 -92.5831 -88.4461>}]
 foot_l [{shapeType=Capsule, length=15.1419, radius=8.1761, center=<-3.2817 8.2665 -0.3323>, rotation=<rot 0 7.3934 -88.4688>}]
 foot_r [{shapeType=Capsule, length=15.1419, radius=8.1761, center=<3.2818 -8.2665 0.3323>, rotation=<rot 0 7.3934 -88.4688>}]
 hand_l [{shapeType=Capsule, length=9.0523, radius=6.1206, center=<8.2245 -2.2535 0.0305>, rotation=<rot 0 77.9566 -77.2037>}]
 hand_r [{shapeType=Capsule, length=9.0523, radius=6.1206, center=<-8.2245 2.2535 -0.0305>, rotation=<rot 0 77.9566 -77.2037>}]
 lowerarm_l [{shapeType=Capsule, length=12.6206, radius=8.1948, center=<11.1691 0.6294 1.2325>, rotation=<rot 0 -89.8949 -82.401>}]
 lowerarm_r [{shapeType=Capsule, length=12.6206, radius=8.1948, center=<-11.1691 -0.6294 -1.2325>, rotation=<rot 0 -89.8949 -82.401>}]
 neck_02 [{shapeType=Capsule, length=13.9709, radius=13.2448, center=<10.9503 1.154 0.01>, rotation=<rot 0 28.2607 -89.4751>}]
 pelvis [{shapeType=Capsule, length=31.3884, radius=24.5271, center=<-15.7219 1.3872 1.2261>, rotation=<rot 0 83.0765 -72.7552>}]
 spine_01 [{shapeType=Capsule, length=16.8779, radius=14.5123, center=<5.2339 1.722 0.7004>, rotation=<rot -3.8218 -90 -0.0274>}]
 spine_02 [{shapeType=Capsule, length=17.6116, radius=15.934, center=<2.0348 2.3164 0>, rotation=<rot 90 95.1417 -174.858>}]
 spine_03 [{shapeType=Capsule, length=18.8307, radius=17.8113, center=<2.085 1.9218 0>, rotation=<rot 90 98.5722 -171.4277>}]
 spine_04 [{shapeType=Capsule, length=19.4842, radius=17.6556, center=<3.5667 2.3523 0>, rotation=<rot 90 96.0453 -173.9546>}]
 spine_05 [{shapeType=Capsule, length=21.6504, radius=15.1844, center=<6.8021 -0.7818 0.0005>, rotation=<rot -0.1713 -90 0.0093>}]
 thigh_twist_01_l [{shapeType=Capsule, length=16.3909, radius=13.4625, center=<8.4729 1.5226 -1.6703>, rotation=<rot 0 -62.7604 -64.3573>}]
 thigh_twist_01_r [{shapeType=Capsule, length=16.4501, radius=13.4375, center=<-8.1604 -1.3271 2.509>, rotation=<rot 0 -63.1204 -66.1439>}]
 upperarm_l [{shapeType=Capsule, length=17.4675, radius=9.8092, center=<11.5116 -1.0945 0.833>, rotation=<rot 0 -94.1567 -82.614>}]
 upperarm_r [{shapeType=Capsule, length=17.4888, radius=9.8449, center=<-11.4974 1.0976 -0.8373>, rotation=<rot 0 -94.4294 -82.5441>}]
}
constraints [
 {bone1Name=spine_01, bone2Name=pelvis, swing1LimitDegrees=45, swing2LimitDegrees=45, twistLimitDegrees=45}
 {bone1Name=spine_02, bone2Name=spine_01, swing1LimitDegrees=45, swing2LimitDegrees=45, twistLimitDegrees=45}
 {bone1Name=spine_03, bone2Name=spine_02, swing1LimitDegrees=45, swing2LimitDegrees=45, twistLimitDegrees=45}
 {bone1Name=spine_04, bone2Name=spine_03, swing1LimitDegrees=45, swing2LimitDegrees=45, twistLimitDegrees=45}
 {bone1Name=spine_05, bone2Name=spine_04, swing1LimitDegrees=45, swing2LimitDegrees=45, twistLimitDegrees=45}
 {bone1Name=neck_02, bone2Name=spine_05, swing1LimitDegrees=45, swing2LimitDegrees=45, twistLimitDegrees=45}
 {bone1Name=upperarm_l, bone2Name=spine_05, swing1LimitDegrees=45, swing2LimitDegrees=45, twistLimitDegrees=45}
 {bone1Name=lowerarm_l, bone2Name=upperarm_l, swing1LimitDegrees=45, swing2LimitDegrees=45, twistLimitDegrees=45}
 {bone1Name=hand_l, bone2Name=lowerarm_l, swing1LimitDegrees=45, swing2LimitDegrees=45, twistLimitDegrees=45}
 {bone1Name=upperarm_r, bone2Name=spine_05, swing1LimitDegrees=45, swing2LimitDegrees=45, twistLimitDegrees=45}
 {bone1Name=lowerarm_r, bone2Name=upperarm_r, swing1LimitDegrees=45, swing2LimitDegrees=45, twistLimitDegrees=45}
 {bone1Name=hand_r, bone2Name=lowerarm_r, swing1LimitDegrees=45, swing2LimitDegrees=45, twistLimitDegrees=45}
 {bone1Name=calf_l, bone2Name=pelvis, swing1LimitDegrees=45, swing2LimitDegrees=45, twistLimitDegrees=45}
 {bone1Name=foot_l, bone2Name=calf_l, swing1LimitDegrees=45, swing2LimitDegrees=45, twistLimitDegrees=45}
 {bone1Name=thigh_twist_01_l, bone2Name=pelvis, swing1LimitDegrees=45, swing2LimitDegrees=45, twistLimitDegrees=45}
 {bone1Name=calf_r, bone2Name=pelvis, swing1LimitDegrees=45, swing2LimitDegrees=45, twistLimitDegrees=45}
 {bone1Name=foot_r, bone2Name=calf_r, swing1LimitDegrees=45, swing2LimitDegrees=45, twistLimitDegrees=45}
 {bone1Name=thigh_twist_01_r, bone2Name=pelvis, swing1LimitDegrees=45, swing2LimitDegrees=45, twistLimitDegrees=45}
]
